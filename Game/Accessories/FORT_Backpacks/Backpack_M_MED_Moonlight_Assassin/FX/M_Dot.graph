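# Additive unlit particle sprite material for the Moonlight Assassin backpack FX, drawing a soft round dot.
# A UV sphere mask raised to an exponent is tinted by particle color and passed through a shared emissive-adjustment function.
# With UseWpo on, non-mobile sprites grow randomly with camera distance and collapse as particles move away from the object origin.
# With it off, world position offset enlarges sprites near the origin, fading with distance.
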Material /Game/Accessories/FORT_Backpacks/Backpack_M_MED_Moonlight_Assassin/FX/M_Dot
graph {
 properties {
  bUseMaterialAttributes=true
  bUsedWithNiagaraSprites=true
  bUsedWithParticleSprites=true
  blendMode=BLEND_Additive
  pixelDepthOffsetMode=PDOM_Legacy
  shadingModel=MSM_Unlit
  usedShadingModels=Unlit
 }
 nodes [
  n1 = Constant() {r=0.5}
  n2 = SphereMask(A=TextureCoordinate(), B=@n1, Radius=@n1, Hardness=Constant())
  n3 = ScalarParameter() {defaultValue=3, parameterName=Exp}
  n4 = ParticleColor()
  n5 = SetMaterialAttributes!(MaterialAttributes, Emissive Color=Multiply(A=Power(Base=@n2, Exp=@n3), B=@n4.RGB), Opacity=@n4.A) {attributeSetTypes [#1, #2]}
  n6 = ParticlePositionWS()
  n7 = Subtract(A=WorldPosition().XYZ, B=@n6)
  n8 = Multiply(A=@n7, B=LinearInterpolate(Alpha=ParticleRandom()) {constB=2.5})
  n9 = CameraPositionWS()
  n10 = Saturate(None=Divide(A=Distance(A=@n6, B=@n9)) {constB=500})
  n11 = Divide(A=Distance(A=ObjectPositionWS(), B=ParticlePositionWS())) {constB=100}
  n12 = SetMaterialAttributes!(
   MaterialAttributes=@n5
   World Position Offset=LinearInterpolate(A=LinearInterpolate(B=@n8, Alpha=@n10) {constB=0}, B=Multiply(A=@n7) {constB=-1}, Alpha=Saturate(None=@n11))
  ) {attributeSetTypes [#3]}
  n13 = LinearInterpolate(A=@n7, Alpha=Saturate(None=Subtract(A=@n11) {constB=0.5})) {constB=0}
  MaterialFunctionCall!(In=StaticSwitchParameter(
    True=ShadingPathSwitch(Default=@n12, Mobile=@n5) {desc=Migrated from FeatureLevelSwitch}
    False=SetMaterialAttributes(MaterialAttributes=@n5, World Position Offset=@n13) {attributeSetTypes [#3]}
   ) {defaultValue=true, parameterName=UseWpo}) {materialFunction=/Game/Effects/Niagara/MasterMaterials/Functions/MF_FX_EmissiveAdjustments_MA}
 ]
}
