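# Translucent unlit particle material for the wave effect of the Moonlight Assassin backpack.
# A time-driven swirl of the UVs, distorted by panning noise through dynamic parameters, forms a sphere mask shaped by a curve-atlas ramp and optional sine.
# Particle color tints emissive and opacity, with depth fade outside mobile, before a shared emissive-adjustment function.
# The UseSpirte switch adds UV1-based time offsets and builds pulsing, camera-facing quads through world position offset.
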Material /Game/Accessories/FORT_Backpacks/Backpack_M_MED_Moonlight_Assassin/FX/M_Waves
graph {
 properties {
  bUseMaterialAttributes=true
  bUsedWithMeshParticles=true
  bUsedWithNiagaraSprites=true
  bUsedWithParticleSprites=true
  blendMode=BLEND_Translucent
  pixelDepthOffsetMode=PDOM_Legacy
  shadingModel=MSM_Unlit
  usedShadingModels=Unlit
 }
 nodes [
  n1 = ParticleColor()
  n2 = Constant() {r=0.5}
  n3 = SphereMask(A=TextureCoordinate(), B=@n2, Radius=@n2, Hardness=Constant())
  n4 = Time()
  n5 = ComponentMask(None=TextureCoordinate() {coordinateIndex=1}) {g=true}
  n6 = StaticSwitchParameter(True=Add(A=@n4, B=Multiply(A=@n5) {constB=32}), False=@n4) {parameterName=UseSpirte}
  n7 = Multiply(A=Add(A=Multiply(A=@n3) {constB=4}, B=@n6), B=ScalarParameter() {defaultValue=0.5, parameterName=Speed_Rotation})
  n8 = Cosine(None=@n7)
  n9 = Subtract(A=TextureCoordinate()) {constB=0.5}
  n10 = ComponentMask(None=@n9) {r=true}
  n11 = Sine(None=@n7)
  n12 = ComponentMask(None=@n9) {g=true}
  n13 = AppendVector(A=Subtract(A=Multiply(A=@n8, B=@n10), B=Multiply(A=@n11, B=@n12)), B=Add(A=Multiply(A=@n10, B=@n11), B=Multiply(A=@n12, B=@n8)))
  n14 = VectorParameter() {parameterName=UV, defaultValue=<rgba 1 1 1 0>}
  n15 = ScalarParameter() {defaultValue=0.25, parameterName=PannerSpeed}
  n16 = DynamicParameter() {defaultValue=<rgba 1 1 0 1>, paramNames [Time,Random,Distort,Tick_Offset]}
  n17 = Add(
   A=Multiply(A=TextureCoordinate(), B=ComponentMask(None=@n14.RGB) {g=true, r=true})
   B=Add(A=Multiply(A=@n6, B=@n15) {constB=0.25}, B=@n16.Random)
  )
  n18 = ComponentMask(None=Subtract(A=TextureSample(UVs=@n17) {samplerType=SAMPLERTYPE_LinearColor, texture=&T_NoisePack_LinearColor}.RGB) {constB=0.5}) {g=true, r=true}
  n19 = Constant() {r=0.5}
  n20 = SphereMask(A=Add(A=Add(A=@n13) {constB=0.48}, B=Multiply(A=@n18, B=@n16.Distort) {constB=0.04}), B=@n19, Radius=@n19, Hardness=Constant())
  n21 = CurveAtlasRowParameter(CurveTime=@n20) {
   atlas=&AthenaCurveAtlas01
   curve=&/Curve_Ramps
   defaultValue=72
  }
  n22 = Multiply(A=@n20) {constB=0.5}
  n23 = Multiply(
   A=StaticComponentMaskParameter(None=AppendVector(A=@n21, B=@n21)) {defaultA=true, parameterName=Mask}
   B=StaticSwitchParameter(True=Sine(None=@n22), False=@n20) {defaultValue=true, parameterName=UseSinMask}
  )
  n24 = Multiply(A=@n1.A, B=Multiply(A=@n23) {constB=0.1})
  n25 = DepthFade(FadeDistance=ScalarParameter() {defaultValue=2, parameterName=DF_FD})
  n26 = SetMaterialAttributes!(
   MaterialAttributes
   Emissive Color=Multiply(A=@n1.RGB, B=@n23)
   Opacity=ShadingPathSwitch(Default=Multiply(A=@n24, B=@n25), Mobile=@n24) {desc=Migrated from FeatureLevelSwitch}
  ) {attributeSetTypes [#1, #2]}
  n27 = Multiply(A=Time()) {constB=0.2}
  n28 = ComponentMask(None=TextureCoordinate() {coordinateIndex=1}) {g=true}
  n29 = LinearInterpolate(Alpha=Sine(None=Add(A=@n27, B=@n28))) {constA=0.75}
  n30 = TextureCoordinate() {coordinateIndex=1, uTiling=0.5}
  n31 = Add(A=Sine(None=ComponentMask(None=@n30) {r=true})) {constB=0.5}
  n32 = MaterialFunctionCall(Float2=Subtract(A=TextureCoordinate()) {constB=0.5}) {materialFunction=&BreakOutFloat2Components}
  n33 = Constant3Vector() {constant=<rgba 1 0 0 0>}
  n34 = Constant3Vector() {constant=<rgba 0 -1 0 0>}
  n35 = Add(
   A=Multiply(A=@n32.R, B=Transform(None=@n33) {transformSourceType=TRANSFORMSOURCE_Camera})
   B=Multiply(A=@n32.G, B=Transform(None=@n34) {transformSourceType=TRANSFORMSOURCE_Camera})
  )
  n36 = ScalarParameter() {defaultValue=25, parameterName=SpriteSize}
  n37 = Add(
   A=MaterialFunctionCall(Offset Amount=ScalarParameter() {defaultValue=25, parameterName=CameraOffset}) {materialFunction=/Engine/Functions/Engine_MaterialFunctions02/WorldPositionOffset/CameraOffset}.Camera Offset
   B=Multiply(A=Multiply(A=@n29, B=@n31), B=Multiply(A=@n35, B=@n36))
  )
  MaterialFunctionCall!(In=StaticSwitchParameter(
    True=SetMaterialAttributes!(MaterialAttributes=@n26, World Position Offset=@n37) {attributeSetTypes [#3]}
    False=@n26
   ) {parameterName=UseSpirte}) {materialFunction=/Game/Effects/Niagara/MasterMaterials/Functions/MF_FX_EmissiveAdjustments_MA}
  Time()
  VertexColor()
 ]
}
